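; Editor QA action that compares a set of assets against a configurable Data Table to find rows with no matching in-engine asset.
; RunActionOnAssetDataArray calls its parent, then collects unique asset names from the incoming asset data.
; It checks each table row name against those names, optionally printing tracked and untracked rows via separate flags, and prints the untracked total.
; The check is one-way: assets missing from the table are not reported.
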(fn RunActionOnAssetDataArray
  (params (AssetDataArray :type Array%20of%20Asset%20Data%20Structures))
  (bind _returnvalue (|CallParentFunction AssetDataArray))
  (for _array_element AssetDataArray
    (Utilities|Array|AddUnique (LocalVariables|GetTrackedAssetNames) (.AssetName (Utilities|Struct|BreakAssetData _array_element))))
  (bind _outrownames (DataTable|GetDataTableRowNames (Variables|Default|GetDataTableToCompare)))
  (bind _output_get (LocalVariables|SetDataTableRowNames _outrownames))
  ;[1 0.682 0.2371] This compares the Data Table against the Tracked Assets, so we know which DataTable Rows didn't have in-engine assets to check against.
  (for _array_element_2 _output_get
    (if (Utilities|Array|ContainsItem (LocalVariables|GetTrackedAssetNames) _array_element_2)
      (Development|PrintString 0 (Utilities|String|Append "Tracking: " (Utilities|String|ToString(Name) _array_element_2)) false (Variables|Default|GetbPrintTrackedAssets))
      (else
        (Math|Integer|IncrementInt (LocalVariables|GetNumOfUnTrackedAssets)
          (:%20%20
            (Development|PrintString 0 (Utilities|String|Append "ERROR: " (Utilities|String|ToString(Name) _array_element_2) " is NOT a tracked asset. Does this need a bug or sync?") false (Variables|Default|GetbPrintUntrackedAssets)))))))
  (Development|PrintString 0 (Utilities|String|Append "ERROR: Number of untracked asset from Data Table: " (Utilities|String|ToString(Integer) (LocalVariables|GetNumOfUnTrackedAssets))))
  (return _returnvalue))
