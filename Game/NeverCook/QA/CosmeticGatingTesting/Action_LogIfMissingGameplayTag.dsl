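; QA asset action for cosmetic gating tests that reports selected item definitions missing an expected gameplay tag, described as the Teen tag.
; The RunActionOnAssetDataArray override calls its parent, casts each asset to FortItemDefinition and checks its owned tags against a configurable tag.
; Pass and fail messages print behind separate bool toggles, while cast failures print separately.
; A counter tallies missing tags for a final summary, which carries an ERROR prefix even at zero.
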(fn RunActionOnAssetDataArray
  (params (AssetDataArray :type Array%20of%20Asset%20Data%20Structures))
  (bind _returnvalue (|CallParentFunction AssetDataArray))
  (for _array_element AssetDataArray
    ;[1 0.5932 0.11] If the asset is missing the expected Teen Gameplay Tag, log an error. Bool variables can be un/checked to print for desired output.
    (bind _asfort_item_definition (Utilities|Casting|CastToFortItemDefinition (AssetData|GetAsset _array_element))
      (:then
        (if (GameplayTags|HasTag (GameplayTags|GetOwnedGameplayTags _asfort_item_definition :!class /Script/GameplayTags.GameplayTagAssetInterface) (Variables|Default|GetGameplayTag) true)
          (Development|PrintString 0 (Utilities|String|Append "In-game Name: " (Utilities|String|ToString(Text) (Variables|/Script/ItemizationCoreRuntime.ItemDefinitionBase|GetItemName _asfort_item_definition)) " has the expected tag. | Asset ID: " (Utilities|GetDisplayName (AssetData|GetAsset _array_element))) false (Variables|Default|GetbPrintSuccesses))
          (else
            (Math|Integer|IncrementInt (LocalVariables|GetNumOfAssetsMissingTags)
              (:%20%20
                (Development|PrintString 0 (Utilities|String|Append "ERROR: " (Utilities|GetDisplayName (AssetData|GetAsset _array_element)) " is MISSING the Teen Tag. BUG IT!") false (Variables|Default|GetbPrintFailures)))))))
      (:CastFailed
        (Development|PrintString 0 (Utilities|String|Append "ERROR: This failed to cast for a FortItemDefinition: " (Utilities|GetDisplayName (AssetData|GetAsset _array_element)))))))
  (Development|PrintString 0 (Utilities|String|Append "ERROR: Number of assets misssing the expected Gameplay Tag: " (Utilities|String|ToString(Integer) (LocalVariables|GetNumOfAssetsMissingTags))))
  (return _returnvalue))
